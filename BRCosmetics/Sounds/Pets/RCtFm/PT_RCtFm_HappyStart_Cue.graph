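# Sound cue for the start of the RCtFm pet's happy vocalization, part of the BRCosmetics pet sounds.
# A single Random node picks among four equally weighted Happy wave players.
# Output is routed through the creature pets sound class with close-range pet attenuation and a slight volume boost.
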SoundCue /BRCosmetics/Sounds/Pets/RCtFm/PT_RCtFm_HappyStart_Cue
root=Random_0
nodes [
 {id=Random_0, children [WavePlayer_4,WavePlayer_5,WavePlayer_6,WavePlayer_7], properties {weights [1,1,1,1]}}
 {id=WavePlayer_4, properties {soundWaveAssetPtr=&/PT_RCtFm_Happy_01}}
 {id=WavePlayer_5, properties {soundWaveAssetPtr=&/PT_RCtFm_Happy_02}}
 {id=WavePlayer_6, properties {soundWaveAssetPtr=&/PT_RCtFm_Happy_03}}
 {id=WavePlayer_7, properties {soundWaveAssetPtr=&/PT_RCtFm_Happy_04}}
]
properties {
 attenuationSettings=&Fort_Pets_Close_Athena_Att
 maxDistance=1800.0001
 soundClassObject=&Fort_Creatures_Pets
 volumeMultiplier=1.12
}
