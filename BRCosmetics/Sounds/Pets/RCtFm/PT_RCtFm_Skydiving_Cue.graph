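# Sound cue for the RCtFm cosmetic pet's skydiving sound.
# A Random node picks one of four equally weighted skydive wave variations.
# It routes to the creature pets sound class with close-range pet attenuation.
# A concurrency override allows only one instance per owner and blocks new plays while one is active.
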SoundCue /BRCosmetics/Sounds/Pets/RCtFm/PT_RCtFm_Skydiving_Cue
root=Random_0
nodes [
 {id=Random_0, children [WavePlayer_0,WavePlayer_1,WavePlayer_3,WavePlayer_5], properties {weights [1,1,1,1]}}
 {id=WavePlayer_0, properties {soundWaveAssetPtr=&/PT_RCtFm_Skydive_01}}
 {id=WavePlayer_1, properties {soundWaveAssetPtr=&/PT_RCtFm_Skydive_02}}
 {id=WavePlayer_3, properties {soundWaveAssetPtr=&/PT_RCtFm_Skydive_03}}
 {id=WavePlayer_5, properties {soundWaveAssetPtr=&/PT_RCtFm_Skydive_04}}
]
properties {
 attenuationSettings=&Fort_Pets_Close_Athena_Att
 bOverrideConcurrency=true
 maxDistance=1800.0001
 soundClassObject=&Fort_Creatures_Pets
 volumeMultiplier=0.896
 concurrencyOverrides {bLimitToOwner=true, maxCount=1, resolutionRule=PreventNew}
}
